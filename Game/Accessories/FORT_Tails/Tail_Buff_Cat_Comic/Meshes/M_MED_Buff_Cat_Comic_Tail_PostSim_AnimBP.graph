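# Post-sim AnimBP for the Buff Cat Comic tail accessory, adding procedural trailing motion and emote-synced animation.
# A tail sequence, scrubbed to the owner's emote montage position, blends over the input pose per bone, fading with montage activity.
# A ten-bone Trail controller then runs on the result.
# Trail strength drops when jumping or falling and is zero when skydiving, while relaxation speed scales with absolute owner direction.
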AnimBlueprint /Game/Accessories/FORT_Tails/Tail_Buff_Cat_Comic/Meshes/M_MED_Buff_Cat_Comic_Tail_PostSim_AnimBP
target_skeleton=/Game/Accessories/FORT_Tails/Common/Fortnite_Base_Tail_Skeleton
graphs [
 AnimGraph {nodes [
   n1 = Layered blend per bone(
    BasePose=Input Pose()
    BlendPoses_0="Sequence Evaluator (by time) 'Buff_Cat_Comic_Tail'"!(ExplicitTime:float=Get emoteOverwriteJoint_Position() -> (emoteOverwriteJoint_Position:double)) -> (Pose:Pose Link Structure) {is=title AnimGraphNode_SequenceEvaluator, sequence=/Game/Animation/Game/MainPlayer/Emotes/BuffCatComic/Buff_Cat_Comic_Tail}
    BlendWeights_0=Get EmoteAlpha() -> (EmoteAlpha:double)
   ) {layerSetup [{branchFilters [
       {boneName=C_Tail_A_Base_Jnt}
       {boneName=C_Tail_A_1_Jnt}
       {boneName=C_Tail_A_2_Jnt}
       {boneName=C_Tail_A_3_Jnt}
       {boneName=C_Tail_A_4_Jnt}
       {boneName=C_Tail_A_5_Jnt}
       {boneName=C_Tail_A_6_Jnt}
       {boneName=C_Tail_A_7_Jnt}
       {boneName=C_Tail_A_8_Jnt}
       {boneName=C_Tail_A_9_Jnt}
      ]}]}
   n2 = Trail controller - Bone: C_Tail_A_9_Jnt!(
    ComponentPose=Local To Component(LocalPose=@n1)
    Alpha=Get TrailAlpha() -> (TrailAlpha:double)
    RelaxationSpeedScale:float=Get RelaxSpeed() -> (RelaxSpeed:double)
   ) -> (Pose) {
    bLimitRotation=true
    chainLength=10
    maxDeltaTime=0.0333
    alphaScaleBiasClamp {bInterpResult=true}
    rotationLimits [
     {limitMax=<180 180 180>}
     {limitMax=<180 180 180>}
     {limitMax=<180 180 180>}
     {limitMax=<180 180 180>}
     {limitMax=<180 180 180>}
     {limitMax=<180 180 180>}
     {limitMax=<180 180 180>}
     {limitMax=<180 180 180>}
     {limitMax=<180 180 180>}
     {limitMax=<180 180 180>}
    ]
    rotationOffsets [<0 0 0>, <0 0 0>, <0 0 0>, <0 0 0>, <0 0 0>, <0 0 0>, <0 0 0>, <0 0 0>, <0 0 0>, <0 0 0>]
    trailBone {boneName=C_Tail_A_9_Jnt}
    trailRelaxationSpeed {editorCurveData {keys [<at 0 10 RCIM_Linear>,<at 0.5107 8.1304 RCIM_Linear>,<at 1 5 RCIM_Linear>]}}
   }
   Output Pose(Result=Component To Local(ComponentPose=@n2)) {layerGroup=None}
  ]}
 EventGraph {nodes [
   n1 = Event BlueprintUpdateAnimation()
   > n2 = Sequence!(execute) -> (then_0, then_1, then_2:Exec, then_3:Exec)
   > n3 = Branch(Condition=Get bIsSkyDiving())
   > Set TrailAlpha(TrailAlpha:double=0) -> (Output_Get:double)
   n4 = Branch(execute=@n3.else, Condition=OR Boolean(A=Get bIsJumping(), B=Get bIsFalling()))
   > Set TrailAlpha(TrailAlpha=0.25)
   Set TrailAlpha(execute=@n4.else, TrailAlpha=0.5)
   n5 = Get Anim_Inst() -> (Anim_Inst:Anim Instance Object Reference)
   n6 = Montage_IsActive!(
    self:Anim Instance Object Reference=@n5
    Montage:Anim Montage Object Reference=/Game/Animation/Game/MainPlayer/Emotes/BuffCatComic/Emote_BuffCatComic_CMM_M
   ) -> (ReturnValue:bool)
   Set EmoteAlpha(
    execute=@n2.then_2
    EmoteAlpha:double=FInterpTo(Current=Get EmoteAlpha(), Target="To Float (Boolean)"(InBool=@n6), DeltaTime=@n1.DeltaTimeX, InterpSpeed=5)
   ) -> (Output_Get:double)
   > Set emoteOverwriteJoint_Position(emoteOverwriteJoint_Position:double=Montage_GetPosition!(
     self:Anim Instance Object Reference=@n5
     Montage:Anim Montage Object Reference=&Emote_BuffCatComic_CMM_M
    ) -> (ReturnValue:float)) -> (Output_Get:double)
   n7 = GetOwnerDirection!(self:Custom Character Part Anim Instance Object Reference) -> (ReturnValue:float)
   Set RelaxSpeed(execute=@n2.then_1, RelaxSpeed:double=MapRangeClamped(Value="Absolute (Float)"(A=@n7), InRangeA=85, InRangeB=110, OutRangeA=2, OutRangeB=4)) -> (Output_Get:double)
   Event BlueprintInitializeAnimation()
   n8 = GetOwnerSkeletalMeshComponent()
   > Branch(Condition=IsValid(Object=@n8))
   > n9 = Set Skel_Mesh(Skel_Mesh=@n8)
   > Set Anim_Inst(Anim_Inst:Anim Instance Object Reference=GetAnimInstance(self=@n9)) -> (Output_Get:Anim Instance Object Reference)
   TryGetPawnOwner()
  ]}
]
variables [
 Anim_Inst {type=AnimInstance, default=null}
 EmoteAlpha {type=number}
 emoteOverwriteJoint_Position {type=number}
 RelaxSpeed {type=number}
 Skel_Mesh {type=SkeletalMeshComponent}
 TrailAlpha {type=number}
]
